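# Material for the SunRose container gadget, built on the character parent material functions, that animates its lid entirely in vertex offsets.
# The LidAnimation custom primitive data value acts as a timestamp, producing a short normalized, square-root-eased ramp.
# That ramp rotates position and normal about a local-axis hinge, masked by red vertex color.
# A damped sine then adds a vertical squash-and-stretch scaled by height above the pivot, before the PostFX and Creative move-tool functions.
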Material /Game/Gadgets/Assets/SunRose/Materials/M_Container_Sunrose
graph {
 properties {
  bUseMaterialAttributes=true
  bUsedWithClothing=true
  bUsedWithMeshDeformer=true
  bUsedWithNanite=true
  bUsedWithSkeletalMesh=true
  pixelDepthOffsetMode=PDOM_Legacy
  shadingModel=MSM_FromMaterialExpression
  usedShadingModels=Default Lit | Subsurface | Clear Coat | Subsurface Profile | Cloth
 }
 nodes [
  n1 = MaterialFunctionCall() {materialFunction=&MF_CharacterParent_2023_PreFX}
  n2 = GetMaterialAttributes(None=@n1.Result) {attributeGetTypes [#1]}
  n3 = ScalarParameter() {bUseCustomPrimitiveData=true, parameterName=LidAnimation}
  n4 = Subtract(A=Time(), B=@n3)
  n5 = Saturate(None=Divide(A=Abs(None=@n4)) {constB=0.5})
  n6 = Saturate(None=@n3)
  n7 = SquareRoot(None=NamedRerouteDeclaration(None=Multiply(A=@n5, B=@n6)) {name=Time, nodeColor=<rgba 1 0.4922 0 1>})
  n8 = MaterialFunctionCall!(
   RotationAngle=Multiply(A=NamedRerouteDeclaration(None=@n7) {name=Sqwuared Time, nodeColor=<rgba 0 0.6016 1 1>}) {constB=0.07}
   NormalizedRotationAxis=Transform(None=Constant3Vector() {constant=<rgba 1 0 0 1>}) {transformSourceType=TRANSFORMSOURCE_Local}
   PivotPoint=TransformPosition(None=Constant3Vector() {constant=<rgba 0 -30 43 1>}) {transformType=TRANSFORMPOSSOURCE_World}
   Position=WorldPosition().XYZ
   Normal=Transform(None=@n2.Normal)
   TangentU
   TangentV
  ) {materialFunction=/Game/Characters/Player/Male/Medium/Bodies/M_MED_Inferno/FX/Mat/RotateVertexAboutAxis}
  n9 = BlendMaterialAttributes(
   A=@n2.MaterialAttributes
   B=SetMaterialAttributes!(
    MaterialAttributes=@n2.MaterialAttributes
    World Position Offset=@n8.WPO
    Normal=Transform(None=@n8.Normal) {transformSourceType=TRANSFORMSOURCE_World, transformType=TRANSFORM_Tangent}
   ) {attributeSetTypes [#2, #1]}
   Alpha=VertexColor().R
  )
  n10 = GetMaterialAttributes(None=@n9) {attributeGetTypes [#2]}
  n11 = Add(A=@n10.World Position Offset, B=WorldPosition().XYZ)
  n12 = MaterialFunctionCall() {materialFunction=&ObjectPivotPoint}
  n13 = NamedRerouteUsage()
  n14 = OneMinus(None=NamedRerouteUsage())
  n15 = Multiply(A=@n14, B=@n14)
  n16 = Multiply(
   A=ComponentMask(None=Subtract(A=@n11, B=@n12.Object Pivot Location)) {b=true}
   B=Multiply(A=Sine(None=@n13) {period=0.5}, B=Multiply(A=@n13, B=@n15))
  )
  n17 = Add(A=@n10.World Position Offset, B=AppendVector(A=Constant2Vector(), B=Multiply(A=@n16) {constB=0.16}))
  MaterialFunctionCall(MA In=MaterialFunctionCall(
    Material=SetMaterialAttributes!(MaterialAttributes=@n9, World Position Offset=@n17) {attributeSetTypes [#2]}
    M=@n1.M
    Pre Skinned Local Position=@n1.M
    Distance Blend=@n1.M
    Fresnel=@n1.M
   ) {materialFunction=&MF_CharacterParent_2023_PostFX}.Result) {materialFunction=&MF_CreativeMoveTool}
 ]
}
